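# GradFrom3DDeriv is a library-exposed material function that computes a 3D gradient vector from the results of Prepare3DDeriv and Compute3DDeriv.
# It takes three inputs, the X, Y and Z derivatives, and keeps only the red channel of each.
# AppendMany packs those scalars into an RGB vector, which is returned as the Gradient output without normalization.
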MaterialFunction /Engine/Functions/Engine_MaterialFunctions03/Procedurals/GradFrom3DDeriv
graph {
 properties {bExposeToLibrary=true, description=Compute 3D gradient vector from result of Prepare3DDeriv/Compute3DDeriv}
 nodes [
  n1 = FunctionInput() {description=Derivative in X direction, inputName=dx}
  n2 = FunctionInput() {description=Derivative in Y direction, inputName=dy, sortPriority=1}
  n3 = FunctionInput() {description=Derivative in Z direction, inputName=dz, sortPriority=2}
  FunctionOutput(None=MaterialFunctionCall(R=ComponentMask(None=@n1) {r=true}, G=ComponentMask(None=@n2) {r=true}, B=ComponentMask(None=@n3) {r=true}) {materialFunction=&AppendMany}.RGB) {description=Computed 3D gradient vector direction, outputName=Gradient}
 ]
}
