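# Material function, exposed to the library, that derives a normal from a heightmap animated by four-way chaotic panning, flagged as expensive.
# It runs Motion_4WayChaos three times on an internal texture object: at the input coordinates and offset by Height Bias in U and V.
# Red-channel differences, scaled by Height, form two tangent vectors whose cross product is output without normalization.
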MaterialFunction /Engine/Functions/Engine_MaterialFunctions03/Procedurals/NormalFromHeightmapChaos
graph {
 properties {
  bExposeToLibrary=true
  description="EXPENSIVE - Takes a heightmap input, pans it 4 ways and generates a normal map from it."
  libraryCategoriesText [Misc,Procedurals]
 }
 nodes [
  n1 = Constant2Vector() {r=1}
  n2 = FunctionInput(Preview=Constant() {r=0.005}) {
   bUsePreviewValueAsDefault=true
   description=Spread used to derive the height from a scalar input. Default is .005
   inputName=Height Bias
   inputType=FunctionInput_Scalar
   previewValue=<0.005 0 0 1>
  }
  n3 = Constant()
  n4 = FunctionInput(Preview=TextureCoordinate()) {bUsePreviewValueAsDefault=true, inputName=Coordinates, inputType=FunctionInput_Vector2}
  n5 = TextureObject()
  n6 = MaterialFunctionCall!(Coordinates=Add(A=AppendVector(A=@n2, B=@n3), B=@n4), Speed, Divisor, Texture=@n5) {materialFunction=/Engine/Functions/Engine_MaterialFunctions03/ComplexMotion/Motion_4WayChaos}
  n7 = ComponentMask(None=MaterialFunctionCall(Coordinates=@n4, Texture=@n5) {materialFunction=&Motion_4WayChaos}.Result) {r=true}
  n8 = FunctionInput(Preview=Constant() {r=8}) {
   bUsePreviewValueAsDefault=true
   description="How strong the normal map will be.  Default is 8."
   inputName=Height
   inputType=FunctionInput_Scalar
  }
  n9 = Multiply(A=Subtract(A=ComponentMask(None=@n6.Result) {r=true}, B=@n7), B=@n8)
  n10 = Constant2Vector() {g=1}
  n11 = MaterialFunctionCall(Coordinates=Add(A=@n4, B=AppendVector(A=@n3, B=@n2)), Texture=@n5) {materialFunction=&Motion_4WayChaos}
  n12 = Multiply(A=@n8, B=Subtract(A=ComponentMask(None=@n11.Result) {r=true}, B=@n7))
  FunctionOutput(None=CrossProduct(A=AppendVector(A=@n1, B=@n9), B=AppendVector(A=@n10, B=@n12)))
 ]
}
