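# Material function that derives a soft radial falloff mask from a pixel's position relative to the object's position.
# The object-to-pixel world vector gets a scaled U/V offset, is divided by Falloff Scale and is optionally transformed into local space.
# Static switches pick two channels as a projection plane, which aspect ratios stretch before a SphereMask with adjustable hardness outputs the result.
# Separate switches let the offset and projection each ignore the object's local transform.
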MaterialFunction /Engine/Functions/Engine_MaterialFunctions03/Procedurals/ObjectSpaceFalloff
graph {
 properties {bExposeToLibrary=true, description=Derives a falloff based on object's location., libraryCategoriesText [Misc,Procedurals]}
 nodes [
  n1 = ObjectPositionWS()
  n2 = WorldPosition()
  n3 = AppendVector(A=Constant(), B=FunctionInput() {bUsePreviewValueAsDefault=true, inputName=Offset - U, inputType=FunctionInput_Scalar})
  n4 = FunctionInput() {bUsePreviewValueAsDefault=true, inputName=Offset - V, inputType=FunctionInput_Scalar}
  n5 = FunctionInput(Preview=Constant() {r=64}) {bUsePreviewValueAsDefault=true, inputName=Falloff Scale, inputType=FunctionInput_Scalar, previewValue=<64 0 0 1>}
  n6 = Multiply(A=Multiply(A=AppendVector(A=@n3, B=@n4), B=Constant() {r=2}), B=@n5)
  n7 = Transform(None=@n6) {transformSourceType=TRANSFORMSOURCE_Local}
  n8 = FunctionInput(Preview=StaticBool()) {
   bUsePreviewValueAsDefault=true
   description="If true, ignores object's local transform when applying offsets."
   inputName=Offset Transform Switch
   inputType=FunctionInput_StaticBool
  }
  n9 = Divide(A=Add(A=Subtract(A=@n1, B=@n2.XYZ), B=StaticSwitch(True=@n6, False=@n7, Value=@n8)), B=@n5)
  n10 = FunctionInput(Preview=StaticBool()) {
   bUsePreviewValueAsDefault=true
   description="If true, ignores object's local transform when generating results."
   inputName=Projection Transform Switch
   inputType=FunctionInput_StaticBool
  }
  n11 = StaticSwitch(True=@n9, False=Transform(None=@n9) {transformSourceType=TRANSFORMSOURCE_World, transformType=TRANSFORM_Local}, Value=@n10)
  n12 = StaticBool()
  n13 = FunctionInput(Preview=@n12) {
   bUsePreviewValueAsDefault=true
   description=Swap between using R and G for 2nd coordinate for spheremask.
   inputName=Invert 1st Channel
   inputType=FunctionInput_StaticBool
  }
  n14 = FunctionInput(Preview=@n12) {
   bUsePreviewValueAsDefault=true
   description=Swap between using B and G for 2nd coordinate for spheremask.
   inputName=Invert 2nd channel
   inputType=FunctionInput_StaticBool
  }
  n15 = AppendVector(
   A=StaticSwitch(True=ComponentMask(None=@n11) {r=true}, False=ComponentMask(None=@n11) {g=true}, Value=@n13)
   B=StaticSwitch(True=ComponentMask(None=@n11) {g=true}, False=ComponentMask(None=@n11) {b=true}, Value=@n14)
  )
  n16 = FunctionInput() {
   bUsePreviewValueAsDefault=true
   description=Bias U shape of results.
   inputName=Aspect Ratio - U
   inputType=FunctionInput_Scalar
   previewValue=<1 0 0 1>
  }
  n17 = FunctionInput() {
   bUsePreviewValueAsDefault=true
   description=Bias V shape of results.
   inputName=Aspect Ratio - V
   inputType=FunctionInput_Scalar
   previewValue=<1 0 0 1>
  }
  n18 = AppendVector(A=@n16, B=@n17)
  n19 = FunctionInput() {
   bUsePreviewValueAsDefault=true
   description=0 - 100 defines sharpness of falloff.
   inputName=Falloff Hardness
   inputType=FunctionInput_Scalar
  }
  n20 = Clamp(None=@n19) {maxDefault=100}
  n21 = Constant() {r=100}
  FunctionOutput(None=SphereMask(A=Multiply(A=@n15, B=@n18), B=Constant(), Hardness=Divide(A=@n20, B=@n21)) {attenuationRadius=0.5, hardnessPercent=0})
 ]
}
