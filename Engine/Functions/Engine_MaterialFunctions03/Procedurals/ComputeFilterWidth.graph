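# ComputeFilterWidth is a library-exposed material function that computes a screen-space filter width for an input value, as used in procedural materials.
# It takes the DDX and DDY derivatives of the input and dots each with itself to get squared lengths.
# An If node selects the larger of the two, and a square root turns it into the output width.
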MaterialFunction /Engine/Functions/Engine_MaterialFunctions03/Procedurals/ComputeFilterWidth
graph {
 properties {bExposeToLibrary=true}
 nodes [
  n1 = FunctionInput()
  n2 = DDX(Value=@n1)
  n3 = DotProduct(A=@n2, B=@n2)
  n4 = DDY(Value=@n1)
  n5 = DotProduct(A=@n4, B=@n4)
  FunctionOutput(None=SquareRoot(None=If(A=@n3, B=@n5, "A > B"=@n3, "A == B"=@n5, "A < B"=@n5) {desc=Max}))
 ]
}
