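# GeneratedOffsetBands is a library material function that procedurally generates horizontal or vertical bands from texture coordinates.
# It scales the UVs by Bands, applies a sine, adds Offset to slide the bands, and takes the absolute value.
# A static switch picks the red or green channel to set direction.
# A SphereMask serves as a one-dimensional threshold, comparing that channel to Compare with Width as radius and Sharpness as hardness.
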MaterialFunction /Engine/Functions/Engine_MaterialFunctions03/Procedurals/GeneratedOffsetBands
graph {
 properties {
  bExposeToLibrary=true
  description=Procedurally generates a horizontal or vertical band from the default TextureCoordinates.
  libraryCategoriesText [Misc,Procedurals]
 }
 nodes [
  n1 = FunctionInput(Preview=TextureCoordinate()) {
   bUsePreviewValueAsDefault=true
   description="Texture Coordinates used to generate the band, the higher the UV value, the more bands."
   inputName=Input Coordinates
   inputType=FunctionInput_Vector2
  }
  n2 = FunctionInput() {
   bUsePreviewValueAsDefault=true
   description=Modifies the number of bands
   inputName=Bands
   inputType=FunctionInput_Scalar
   previewValue=<1 0 0 1>
  }
  n3 = Multiply(A=@n1, B=@n2)
  n4 = FunctionInput() {bUsePreviewValueAsDefault=true, description=This value will slide the band around., inputName=Offset, inputType=FunctionInput_Scalar}
  n5 = Abs(None=Add(A=Sine(None=@n3), B=@n4))
  n6 = ComponentMask(None=@n5) {r=true}
  n7 = ComponentMask(None=@n5) {g=true}
  n8 = FunctionInput(Preview=StaticBool()) {
   bUsePreviewValueAsDefault=true
   description="Bool to set direction of the band, default is horizontal."
   inputName=Direction Switch
   inputType=FunctionInput_StaticBool
  }
  n9 = FunctionInput(Preview=Constant() {r=0.5}) {
   bUsePreviewValueAsDefault=true
   description="The value the input is compared against to generate the band.  Default is .5"
   inputName=Compare
   inputType=FunctionInput_Scalar
   previewValue=<0.25 0 0 1>
  }
  n10 = FunctionInput() {
   bUsePreviewValueAsDefault=true
   description="The width of the band, default is .25"
   inputName=Width
   inputType=FunctionInput_Scalar
   previewValue=<0.25 0 0 1>
  }
  n11 = FunctionInput() {
   bUsePreviewValueAsDefault=true
   description="Controls the falloff of the band, 100 will be very sharp and aliased."
   inputName=Sharpness
   inputType=FunctionInput_Scalar
   previewValue=<0.25 0 0 1>
  }
  FunctionOutput(None=SphereMask(A=StaticSwitch(True=@n6, False=@n7, Value=@n8), B=@n9, Radius=@n10, Hardness=@n11))
 ]
}
