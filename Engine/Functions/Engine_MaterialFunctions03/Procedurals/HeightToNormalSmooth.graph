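# Material function in the Normals library category that turns per-pixel height in world units into a partially smooth world-space normal.
# It offsets world position along the vertex normal by Height, then derives normalized face normals from DDX/DDY cross products of both offset and original positions.
# Subtracting the original face normal and adding the vertex normal restores smooth shading, without renormalizing the result.
# Its description notes that smooth details require anti-aliasing.
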MaterialFunction /Engine/Functions/Engine_MaterialFunctions03/Procedurals/HeightToNormalSmooth
graph {
 properties {
  bExposeToLibrary=true
  description="Provide perpixel height data and return a partially smooth normal (requires AA for smooth details)"
  libraryCategoriesText [Normals]
 }
 nodes [
  n1 = FunctionInput() {description=Height in world units, inputName=Height, inputType=FunctionInput_Scalar, previewValue=<0.02 0 0 1>}
  n2 = FunctionInput(Preview=VertexNormalWS()) {bUsePreviewValueAsDefault=true, inputName=World Space Vertex Normals, sortPriority=2}
  n3 = FunctionInput(Preview=WorldPosition().XYZ) {bUsePreviewValueAsDefault=true, inputName=Absolute World Position, sortPriority=1}
  n4 = Add(A=Multiply(A=@n1, B=@n2), B=@n3)
  n5 = Normalize(VectorInput=CrossProduct(A=DDX(Value=@n4), B=DDY(Value=@n4)))
  n6 = Normalize(VectorInput=CrossProduct(A=DDX(Value=@n3), B=DDY(Value=@n3)))
  FunctionOutput(None=Add(A=Subtract(A=@n5, B=@n6), B=@n2)) {outputName=WorldSpaceNormal}
 ]
}
