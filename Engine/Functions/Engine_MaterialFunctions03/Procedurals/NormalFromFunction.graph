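# Material function that derives a normal map from an arbitrary height function using finite differences.
# It outputs base coordinates plus copies offset along U and V, and takes the caller's function evaluated at those three points as inputs.
# Height differences, divided by offset over intensity, form two tangent vectors whose cross product is output as an unnormalized Normal.
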MaterialFunction /Engine/Functions/Engine_MaterialFunctions03/Procedurals/NormalFromFunction
graph {
 properties {bExposeToLibrary=true, description=Takes a heightmap input and generates a normal map from it., libraryCategoriesText [Misc,Procedurals]}
 nodes [
  n1 = Constant2Vector() {r=1}
  n2 = FunctionInput() {inputName="Function(UV1)", sortPriority=3}
  n3 = FunctionInput(Preview=Constant() {r=0.005}) {
   bUsePreviewValueAsDefault=true
   description=Spread used to derive the height from a scalar input. Default is .005
   inputName=Height Map UV Offset
   inputType=FunctionInput_Scalar
   sortPriority=1
   previewValue=<0.005 0 0 1>
  }
  n4 = FunctionInput() {
   bUsePreviewValueAsDefault=true
   description="How strong the normal map will be.  Default is 8."
   inputName=Normal Map Intensity
   inputType=FunctionInput_Scalar
   sortPriority=2
   previewValue=<1 0 0 1>
  }
  n5 = Divide(A=@n3, B=@n4)
  n6 = Divide(A=Subtract(A=FunctionInput() {inputName="Function(UV2)", sortPriority=4}, B=@n2), B=@n5)
  n7 = Constant2Vector() {g=1}
  n8 = Divide(A=Subtract(A=FunctionInput() {inputName="Function(UV3)", sortPriority=5}, B=@n2), B=@n5)
  FunctionOutput(None=CrossProduct(A=AppendVector(A=@n1, B=@n6), B=AppendVector(A=@n7, B=@n8))) {outputName=Normal}
  n9 = FunctionInput(Preview=TextureCoordinate()) {bUsePreviewValueAsDefault=true, inputName=Coordinates, inputType=FunctionInput_Vector2}
  FunctionOutput(None=@n9) {outputName=UV1, sortPriority=3}
  n10 = Constant()
  FunctionOutput(None=Add(A=AppendVector(A=@n3, B=@n10), B=@n9)) {outputName=UV2, sortPriority=4}
  FunctionOutput(None=Add(A=@n9, B=AppendVector(A=@n10, B=@n3))) {outputName=UV3, sortPriority=5}
  Multiply(A=@n4, B=@n3)
 ]
}
