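# NormalFromHeightmap is a library material function that generates a normal vector from a heightmap texture.
# It samples the texture at the input coordinates and at small U and V offsets, using a dot product with a selector vector to choose the height channel, red by default.
# The two forward differences, scaled by intensity, form tangent vectors whose cross product is output without a normalize step.
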MaterialFunction /Engine/Functions/Engine_MaterialFunctions03/Procedurals/NormalFromHeightmap
graph {
 properties {bExposeToLibrary=true, description=Takes a heightmap input and generates a normal map from it., libraryCategoriesText [Misc,Procedurals]}
 nodes [
  n1 = Constant2Vector() {r=1}
  n2 = FunctionInput(Preview=Constant() {r=0.005}) {
   bUsePreviewValueAsDefault=true
   description=Spread used to derive the height from a scalar input. Default is .005
   inputName=Height Map UV Offset
   inputType=FunctionInput_Scalar
   sortPriority=2
   previewValue=<0.005 0 0 1>
  }
  n3 = Constant()
  n4 = FunctionInput(Preview=TextureCoordinate()) {bUsePreviewValueAsDefault=true, inputName=Coordinates, inputType=FunctionInput_Vector2, sortPriority=3}
  n5 = FunctionInput(Preview=TextureObject()) {
   bUsePreviewValueAsDefault=true
   description=Uses the R channel as the heightmap by default.
   inputName=Height Map
   inputType=FunctionInput_Texture2D
  }
  n6 = TextureSample(UVs=Add(A=AppendVector(A=@n2, B=@n3), B=@n4), Tex=@n5)
  n7 = AppendVector(A=@n6.RGB, B=@n6.RGB)
  n8 = FunctionInput() {
   bUsePreviewValueAsDefault=true
   description="If you would like the red channel to be used to generate the normal map enter (1,0,0,0). If the green channel is preferred enter (0,1,0,0) etc."
   inputName=Height Map Channel Selector
   inputType=FunctionInput_Vector4
   sortPriority=11
   previewValue=<1 0 0 0>
  }
  n9 = TextureSample(UVs=@n4, Tex=@n5)
  n10 = DotProduct(A=AppendVector(A=@n9.RGB, B=@n9.RGB), B=@n8)
  n11 = FunctionInput(Preview=Constant() {r=8}) {
   bUsePreviewValueAsDefault=true
   description="How strong the normal map will be.  Default is 8."
   inputName=Normal Map Intensity
   inputType=FunctionInput_Scalar
   sortPriority=1
  }
  n12 = Multiply(A=Subtract(A=DotProduct(A=@n7, B=@n8), B=@n10), B=@n11)
  n13 = Constant2Vector() {g=1}
  n14 = TextureSample(UVs=Add(A=@n4, B=AppendVector(A=@n3, B=@n2)), Tex=@n5)
  n15 = AppendVector(A=@n14.RGB, B=@n14.RGB)
  n16 = Multiply(A=@n11, B=Subtract(A=DotProduct(A=@n15, B=@n8), B=@n10))
  FunctionOutput(None=CrossProduct(A=AppendVector(A=@n1, B=@n12), B=AppendVector(A=@n13, B=@n16)))
 ]
}
