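# PerturbNormalHQ is a library material function that perturbs a world-space normal, vertex normal by default, using three bump height samples.
# It crosses screen-space derivatives of world position with the normal, then weights them by right and down height differences to form a surface gradient.
# The sign-corrected gradient is subtracted from the determinant-scaled normal and normalized, so bump mapping needs no UV tangent basis.
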MaterialFunction /Engine/Functions/Engine_MaterialFunctions03/Procedurals/PerturbNormalHQ
graph {
 properties {bExposeToLibrary=true}
 nodes [
  n1 = WorldPosition()
  n2 = FunctionInput(Preview=VertexNormalWS()) {bUsePreviewValueAsDefault=true, inputName=WS Normal, sortPriority=3}
  n3 = CrossProduct(A=DDY(Value=@n1.XYZ) {desc=sigmaT}, B=@n2) {desc=tR1}
  n4 = DDX(Value=@n1.XYZ) {desc=sigmaS}
  n5 = DotProduct(A=@n3, B=@n4) {desc=fDet}
  n6 = Multiply(A=Abs(None=@n5), B=@n2)
  n7 = FunctionInput() {inputName=Bump center, inputType=FunctionInput_Scalar, sortPriority=1}
  n8 = Subtract(A=FunctionInput() {inputName=Bump one pixel right, inputType=FunctionInput_Scalar}, B=@n7)
  n9 = Subtract(A=FunctionInput() {inputName=Bump one pixel down, inputType=FunctionInput_Scalar, sortPriority=2}, B=@n7)
  n10 = CrossProduct(A=@n2, B=@n4) {desc=tR2}
  n11 = Constant()
  n12 = Multiply(
   A=Add(A=Multiply(A=@n8, B=@n3), B=Multiply(A=@n9, B=@n10))
   B=If(A=@n5, B=@n11, "A > B"=Constant() {r=1}, "A == B"=@n11, "A < B"=Constant() {r=-1}) {desc=Sign, equalsThreshold=0}
  ) {desc=SurfGrad}
  FunctionOutput(None=Normalize(VectorInput=Subtract(A=@n6, B=@n12))) {outputName=WS Normal}
 ]
}
