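# Material function, exposed to the library, that prepares screen-space offset samples of an input Value for high-quality normal perturbation in procedural materials.
# It adds DDX and DDY of Value to itself to output estimates one pixel right and one pixel down, alongside the unchanged Value.
# FilterWidth is the square root of the larger squared derivative length, chosen with an If node acting as Max.
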MaterialFunction /Engine/Functions/Engine_MaterialFunctions03/Procedurals/PreparePerturbNormalHQ
graph {
 properties {bExposeToLibrary=true}
 nodes [
  n1 = FunctionInput() {inputName=Value}
  n2 = DDX(Value=@n1)
  FunctionOutput(None=Add(A=@n1, B=@n2)) {outputName=Value one pixel right}
  FunctionOutput(None=@n1) {outputName=Value, sortPriority=1}
  n3 = DDY(Value=@n1)
  FunctionOutput(None=Add(A=@n1, B=@n3)) {outputName=Value one pixel down, sortPriority=2}
  n4 = DotProduct(A=@n2, B=@n2)
  n5 = DotProduct(A=@n3, B=@n3)
  FunctionOutput(None=SquareRoot(None=If(A=@n4, B=@n5, "A > B"=@n4, "A < B"=@n5) {desc=Max})) {outputName=FilterWidth, sortPriority=3}
 ]
}
